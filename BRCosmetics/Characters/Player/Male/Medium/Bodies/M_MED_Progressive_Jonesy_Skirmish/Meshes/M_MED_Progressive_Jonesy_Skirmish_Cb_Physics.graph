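# Physics asset for the male medium Progressive Jonesy Skirmish cosmetic body, apparently for secondary motion of a crossbow on the left thigh.
# Capsules cover only pelvis, spine and left leg, constrained as a chain.
# A tiny master capsule links to the thigh with free rotation and anchors the two-capsule crossbow body through narrower swing and twist limits.
# The thigh-to-pelvis joint also rotates freely.
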PhysicsAsset /BRCosmetics/Characters/Player/Male/Medium/Bodies/M_MED_Progressive_Jonesy_Skirmish/Meshes/M_MED_Progressive_Jonesy_Skirmish_Cb_Physics
bodies {
 calf_l [{shapeType=Capsule, length=21.0995, radius=7.8762, center=<-19.2727 -1.1779 0.1732>, rotation=<rot 0 86.8317 -86.1035>}]
 dyn_crossbow_01_l_MSTR [{shapeType=Capsule, length=0.01, radius=0.01}]
 dyn_crossbow_02_l [
  {shapeType=Capsule, length=14.519, radius=6.6314, center=<4.8929 0 0.0002>, rotation=<rot 0 0 89.9987>}
  {shapeType=Capsule, length=5.817, radius=7.5, center=<5.4633 17.8005 -1.0415>}
 ]
 foot_l [{shapeType=Capsule, length=15.3591, radius=8.7988, center=<-3.7292 7.8264 0.0013>, rotation=<rot 0 12.741 -88.3735>}]
 pelvis [{shapeType=Capsule, length=12.25, radius=10}]
 spine_02 [{shapeType=Capsule, length=14.1167, radius=13.9692, center=<6.7564 -0.8363 0.1761>, rotation=<rot 0 -173.9483 -138.4727>}]
 spine_04 [{shapeType=Capsule, length=16.8834, radius=13.9086, center=<3.4683 -0.0074 0.9627>, rotation=<rot -12.6163 -90 1.8382>}]
 spine_05 [{shapeType=Capsule, length=25.3276, radius=21.1364, center=<-5.6106 0.156 -0.1934>, rotation=<rot 4.9422 -90 -89.3311>}]
 thigh_l [{shapeType=Capsule, length=24.4802, radius=6.7694, center=<-18.2587 0.0001 0.7243>, rotation=<rot 90 0 0>}]
}
constraints [
 {bone1Name=spine_02, bone2Name=pelvis, swing1LimitDegrees=45, swing2LimitDegrees=45, twistLimitDegrees=45}
 {bone1Name=spine_04, bone2Name=spine_02, swing1LimitDegrees=45, swing2LimitDegrees=45, twistLimitDegrees=45}
 {bone1Name=spine_05, bone2Name=spine_04, swing1LimitDegrees=45, swing2LimitDegrees=45, twistLimitDegrees=45}
 {bone1Name=calf_l, bone2Name=thigh_l, swing1LimitDegrees=45, swing2LimitDegrees=45, twistLimitDegrees=45}
 {bone1Name=foot_l, bone2Name=calf_l, swing1LimitDegrees=45, swing2LimitDegrees=45, twistLimitDegrees=45}
 {bone1Name=dyn_crossbow_02_l, bone2Name=dyn_crossbow_01_l_MSTR, swing1LimitDegrees=45, swing2LimitDegrees=15, twistLimitDegrees=5}
 {
  bone1Name=dyn_crossbow_01_l_MSTR
  bone2Name=thigh_l
  swing1LimitDegrees=45
  swing1Motion=Free
  swing2LimitDegrees=45
  swing2Motion=Free
  twistLimitDegrees=45
  twistMotion=Free
 }
 {
  bone1Name=thigh_l
  bone2Name=pelvis
  swing1LimitDegrees=45
  swing1Motion=Free
  swing2LimitDegrees=45
  swing2Motion=Free
  twistLimitDegrees=45
  twistMotion=Free
 }
]
