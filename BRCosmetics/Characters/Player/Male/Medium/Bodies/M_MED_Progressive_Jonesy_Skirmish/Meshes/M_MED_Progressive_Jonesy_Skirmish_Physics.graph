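# Physics asset for the male medium Progressive Jonesy Skirmish body, giving dyn_ clothing bones (belts, buckle, shirt, collar, pant legs) bodies and joints for secondary motion.
# Coarse capsules cover pelvis, spine, arms, right thigh and both calf twists; there is no left thigh body, so the left calf joins the pelvis.
# Belt, buckle and shirt pieces hang with limited joints from free-rotating MSTR anchors under the pelvis.
# Collar flaps attach to spine_05, pant pieces to calf twists.
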PhysicsAsset /BRCosmetics/Characters/Player/Male/Medium/Bodies/M_MED_Progressive_Jonesy_Skirmish/Meshes/M_MED_Progressive_Jonesy_Skirmish_Physics
bodies {
 calf_twist_01_l [{shapeType=Capsule, length=8.1666, radius=3.2474, center=<-0.0985 -0.5585 0>, rotation=<rot 90 -1.4321 8.569>}]
 calf_twist_01_r [{shapeType=Capsule, length=8.4767, radius=3.0969, center=<-0.0553 0.4768 0>, rotation=<rot 90 0 10.0007>}]
 clavicle_l [{shapeType=Capsule, length=9.7203, radius=5.7303, center=<8.4231 -0.5029 1.5882>, rotation=<rot -90 0.0197 -0.0197>}]
 clavicle_r [{shapeType=Capsule, length=10.2882, radius=6, center=<-9.2103 0 -1.5882>, rotation=<rot -90 -0.0017 -179.9982>}]
 dyn_belt_knot_MSTR [{shapeType=Box, extentX=1.01, extentY=1.01, extentZ=1.01}]
 dyn_belt_knot_l_MSTR [{shapeType=Box, extentX=1.01, extentY=1.01, extentZ=1.01}]
 dyn_belt_knot_r_MSTR [{shapeType=Box, extentX=1.01, extentY=1.01, extentZ=1.01}]
 dyn_belt_l_01 [{shapeType=Box, extentX=4.7654, extentY=4.1833, extentZ=1.01, center=<2.1588 0 0.0001>}]
 dyn_belt_l_02 [{shapeType=Box, extentX=6.125, extentY=4.2302, extentZ=1.8832, center=<2.0093 0 0.1506>, rotation=<rot 5 0 0>}]
 dyn_belt_l_03 [{shapeType=Box, extentX=3.9503, extentY=4.3749, extentZ=1.6137, center=<1.6058 0 0>}]
 dyn_belt_r_01 [{shapeType=Capsule, length=5.0986, radius=1.4968, shapeName=dyn_belt_r_01_capsule, center=<3.9943 0 0>, rotation=<rot 90 90 90>}]
 dyn_belt_r_02 [{shapeType=Capsule, length=5.0986, radius=1.4968, shapeName=dyn_belt_r_02_capsule, center=<3.9943 0 0>, rotation=<rot 90 0.0002 0.0002>}]
 dyn_belt_r_03 [{shapeType=Capsule, length=5.0986, radius=1.4968, shapeName=dyn_belt_r_03_capsule, center=<3.9943 0 0>, rotation=<rot 90 179.9802 179.9802>}]
 dyn_beltbuckle_01_l [{shapeType=Capsule, length=5.269, radius=2.3814, center=<2.494 0 0.0002>, rotation=<rot 0.0013 89.9994 89.9999>}]
 dyn_beltbuckle_01_l_MSTR [{shapeType=Capsule, length=0.01, radius=0.01}]
 dyn_collarOuter_bk_01_l [{shapeType=Capsule, length=9.9749, radius=3.6875, center=<5 0 0>, rotation=<rot 0 0 90>}]
 dyn_collarOuter_bk_01_r [{shapeType=Capsule, length=9.9749, radius=3.6875, center=<5 -0.2876 -0.6169>, rotation=<rot 0 0 114.9987>}]
 dyn_pants_l_01 [{shapeType=Capsule, length=4.4104, radius=2.3417, center=<4.3328 0 2>, rotation=<rot 90 0 0>}]
 dyn_pants_l_02 [{shapeType=Capsule, length=4.0585, radius=1.548, center=<3.834 0 0>, rotation=<rot 90 0 0>}]
 dyn_pants_l_03 [{shapeType=Capsule, length=3.3284, radius=1.7512, center=<3.6062 0 0>, rotation=<rot 90 0 0>}]
 dyn_pants_l_04 [{shapeType=Capsule, length=3.5972, radius=1.6072, center=<3.1722 0 0.5>, rotation=<rot 90 0 0>}]
 dyn_pants_r_01 [{shapeType=Capsule, length=3.2449, radius=1.6936, center=<3.6409 0 1>, rotation=<rot 90 0 0>}]
 dyn_pants_r_02 [{shapeType=Capsule, length=3.6516, radius=1.6882, center=<3.2833 0 0.5>, rotation=<rot 90 0 0>}]
 dyn_pants_r_03 [{shapeType=Capsule, length=2.8763, radius=2.2931, center=<3.833 0 0>, rotation=<rot 90 0 0>}]
 dyn_pants_r_04 [{shapeType=Capsule, length=5.4666, radius=1.9757, center=<4.9025 0 0>, rotation=<rot 90 0 0>}]
 dyn_shirt_01_r [{shapeType=Box, extentX=12.637, extentY=13.0328, extentZ=4.2891, center=<6.3255 0 0>}]
 dyn_shirt_01_r_MSTR [{shapeType=Capsule, length=0.01, radius=0.01}]
 dyn_shirt_bk_01 [{shapeType=Capsule, length=4.5982, radius=3.0522, center=<3.2085 0 0>, rotation=<rot 0 90 90>}]
 dyn_shirt_bk_01_MSTR [{shapeType=Capsule, length=0.01, radius=0.01}]
 dyn_shirt_bk_r_01 [{shapeType=Capsule, length=7.8695, radius=3.7573, center=<8 0 0>, rotation=<rot 0 90 90>}]
 dyn_shirt_bk_r_01_MSTR [{shapeType=Capsule, length=0.505, radius=0.505, rotation=<rot 0 -90 0>}]
 dyn_shirt_fr_01_r [{shapeType=Box, extentX=16.3047, extentY=2.2988, extentZ=16.2866, center=<5.085 0.3827 0.226>, rotation=<rot 0 0 93.6014>}]
 dyn_shirt_fr_01_r_MSTR [{shapeType=Box, extentX=5.7446, extentY=5.3555, extentZ=0.5, center=<4.5183 -1.0765 0.3953>, rotation=<rot 0 -89.9999 -5.0001>}]
 lowerarm_l [{shapeType=Capsule, length=17.5, radius=4.75, center=<11.8488 0 0>, rotation=<rot -90 0.0264 -0.0263>}]
 lowerarm_r [{shapeType=Capsule, length=17.5, radius=4.75, center=<-11.8488 0 0>, rotation=<rot -90 -179.9802 -0.0197>}]
 pelvis [{shapeType=Capsule, length=10.6393, radius=11.6107}]
 spine_02 [{shapeType=Capsule, length=11.5, radius=10.75}]
 spine_04 [{shapeType=Capsule, length=11.5, radius=10.75}]
 spine_05 [{shapeType=Capsule, length=12.5, radius=9.75}]
 thigh_r [{shapeType=Capsule, length=33.5, radius=6.75, center=<19.8348 0 0>, rotation=<rot -90 -0.0148 0.0148>}]
 upperarm_l [{shapeType=Capsule, length=15.5, radius=6.75, center=<11.8488 0 0>, rotation=<rot -90 0.0264 -0.0263>}]
 upperarm_r [{shapeType=Capsule, length=15.5, radius=6.75, center=<-11.8488 0 0>, rotation=<rot -90 -179.9802 -0.0197>}]
}
constraints [
 {bone1Name=spine_02, bone2Name=pelvis, swing1LimitDegrees=45, swing2LimitDegrees=45, twistLimitDegrees=45}
 {bone1Name=spine_04, bone2Name=spine_02, swing1LimitDegrees=45, swing2LimitDegrees=45, twistLimitDegrees=45}
 {bone1Name=spine_05, bone2Name=spine_04, swing1LimitDegrees=45, swing2LimitDegrees=45, twistLimitDegrees=45}
 {
  bone1Name=upperarm_l
  bone2Name=clavicle_l
  swing1LimitDegrees=45
  swing1Motion=Free
  swing2LimitDegrees=45
  swing2Motion=Free
  twistLimitDegrees=45
  twistMotion=Free
 }
 {
  bone1Name=upperarm_r
  bone2Name=clavicle_r
  swing1LimitDegrees=45
  swing1Motion=Free
  swing2LimitDegrees=45
  swing2Motion=Free
  twistLimitDegrees=45
  twistMotion=Free
 }
 {
  bone1Name=dyn_shirt_bk_01_MSTR
  bone2Name=pelvis
  swing1LimitDegrees=45
  swing1Motion=Free
  swing2LimitDegrees=45
  swing2Motion=Free
  twistLimitDegrees=45
  twistMotion=Free
 }
 {
  bone1Name=dyn_beltbuckle_01_l_MSTR
  bone2Name=pelvis
  swing1LimitDegrees=45
  swing1Motion=Free
  swing2LimitDegrees=45
  swing2Motion=Free
  twistLimitDegrees=45
  twistMotion=Free
 }
 {
  bone1Name=dyn_shirt_fr_01_r_MSTR
  bone2Name=pelvis
  swing1LimitDegrees=45
  swing1Motion=Free
  swing2LimitDegrees=45
  swing2Motion=Free
  twistLimitDegrees=45
  twistMotion=Free
 }
 {bone1Name=dyn_shirt_bk_01, bone2Name=dyn_shirt_bk_01_MSTR, swing1LimitDegrees=20, swing2LimitDegrees=45, twistLimitDegrees=5}
 {bone1Name=dyn_beltbuckle_01_l, bone2Name=dyn_beltbuckle_01_l_MSTR, swing1LimitDegrees=45, swing2LimitDegrees=45, twistLimitDegrees=5}
 {bone1Name=dyn_shirt_fr_01_r, bone2Name=dyn_shirt_fr_01_r_MSTR, swing1LimitDegrees=15, swing2LimitDegrees=45, twistLimitDegrees=5}
 {
  bone1Name=lowerarm_l
  bone2Name=upperarm_l
  swing1LimitDegrees=45
  swing1Motion=Free
  swing2LimitDegrees=45
  swing2Motion=Free
  twistLimitDegrees=45
  twistMotion=Free
 }
 {
  bone1Name=lowerarm_r
  bone2Name=upperarm_r
  swing1LimitDegrees=45
  swing1Motion=Free
  swing2LimitDegrees=45
  swing2Motion=Free
  twistLimitDegrees=45
  twistMotion=Free
 }
 {
  bone1Name=thigh_r
  bone2Name=pelvis
  swing1LimitDegrees=45
  swing1Motion=Free
  swing2LimitDegrees=45
  swing2Motion=Free
  twistLimitDegrees=45
  twistMotion=Free
 }
 {
  bone1Name=dyn_shirt_01_r_MSTR
  bone2Name=pelvis
  swing1LimitDegrees=45
  swing1Motion=Free
  swing2LimitDegrees=45
  swing2Motion=Free
  twistLimitDegrees=45
  twistMotion=Free
 }
 {
  bone1Name=clavicle_l
  bone2Name=spine_05
  swing1LimitDegrees=45
  swing1Motion=Free
  swing2LimitDegrees=45
  swing2Motion=Free
  twistLimitDegrees=45
  twistMotion=Free
 }
 {
  bone1Name=clavicle_r
  bone2Name=spine_05
  swing1LimitDegrees=45
  swing1Motion=Free
  swing2LimitDegrees=45
  swing2Motion=Free
  twistLimitDegrees=45
  twistMotion=Free
 }
 {
  bone1Name=dyn_belt_knot_MSTR
  bone2Name=pelvis
  swing1LimitDegrees=45
  swing1Motion=Free
  swing2LimitDegrees=45
  swing2Motion=Free
  twistLimitDegrees=45
  twistMotion=Free
 }
 {
  bone1Name=dyn_belt_knot_r_MSTR
  bone2Name=dyn_belt_knot_MSTR
  swing1LimitDegrees=45
  swing1Motion=Free
  swing2LimitDegrees=45
  swing2Motion=Free
  twistLimitDegrees=45
  twistMotion=Free
 }
 {bone1Name=dyn_belt_l_01, bone2Name=dyn_belt_knot_r_MSTR, swing1LimitDegrees=5, swing2LimitDegrees=45, twistLimitDegrees=5}
 {bone1Name=dyn_belt_l_02, bone2Name=dyn_belt_l_01, swing1LimitDegrees=5, swing2LimitDegrees=45, twistLimitDegrees=5}
 {bone1Name=dyn_belt_l_03, bone2Name=dyn_belt_l_02, swing1LimitDegrees=5, swing2LimitDegrees=45, twistLimitDegrees=5}
 {
  bone1Name=dyn_belt_knot_l_MSTR
  bone2Name=dyn_belt_knot_MSTR
  swing1LimitDegrees=45
  swing1Motion=Free
  swing2LimitDegrees=45
  swing2Motion=Free
  twistLimitDegrees=45
  twistMotion=Free
 }
 {bone1Name=dyn_belt_r_01, bone2Name=dyn_belt_knot_l_MSTR, swing1LimitDegrees=20, swing2LimitDegrees=65, twistLimitDegrees=5}
 {bone1Name=dyn_belt_r_02, bone2Name=dyn_belt_r_01, swing1LimitDegrees=20, swing2LimitDegrees=65, twistLimitDegrees=5}
 {bone1Name=dyn_belt_r_03, bone2Name=dyn_belt_r_02, swing1LimitDegrees=20, swing2LimitDegrees=65, twistLimitDegrees=5}
 {bone1Name=calf_twist_01_r, bone2Name=thigh_r, swing1LimitDegrees=15, swing2LimitDegrees=15, twistLimitDegrees=5}
 {bone1Name=dyn_pants_r_04, bone2Name=calf_twist_01_r, swing1LimitDegrees=15, swing2LimitDegrees=15, twistLimitDegrees=5}
 {bone1Name=dyn_pants_r_01, bone2Name=calf_twist_01_r, swing1LimitDegrees=15, swing2LimitDegrees=15, twistLimitDegrees=5}
 {bone1Name=dyn_pants_r_02, bone2Name=calf_twist_01_r, swing1LimitDegrees=15, swing2LimitDegrees=15, twistLimitDegrees=5}
 {bone1Name=dyn_pants_r_03, bone2Name=calf_twist_01_r, swing1LimitDegrees=15, swing2LimitDegrees=15, twistLimitDegrees=5}
 {
  bone1Name=calf_twist_01_l
  bone2Name=pelvis
  swing1LimitDegrees=45
  swing1Motion=Free
  swing2LimitDegrees=45
  swing2Motion=Free
  twistLimitDegrees=45
  twistMotion=Free
 }
 {bone1Name=dyn_pants_l_02, bone2Name=calf_twist_01_l, swing1LimitDegrees=15, swing2LimitDegrees=15, twistLimitDegrees=5}
 {bone1Name=dyn_pants_l_04, bone2Name=calf_twist_01_l, swing1LimitDegrees=15, swing2LimitDegrees=15, twistLimitDegrees=5}
 {bone1Name=dyn_pants_l_01, bone2Name=calf_twist_01_l, swing1LimitDegrees=15, swing2LimitDegrees=15, twistLimitDegrees=5}
 {bone1Name=dyn_pants_l_03, bone2Name=calf_twist_01_l, swing1LimitDegrees=15, swing2LimitDegrees=15, twistLimitDegrees=5}
 {bone1Name=dyn_collarOuter_bk_01_l, bone2Name=spine_05, swing1LimitDegrees=5, swing2LimitDegrees=20, twistLimitDegrees=3}
 {bone1Name=dyn_collarOuter_bk_01_r, bone2Name=spine_05, swing1LimitDegrees=5, swing2LimitDegrees=20, twistLimitDegrees=3}
 {bone1Name=dyn_shirt_bk_r_01, bone2Name=dyn_shirt_bk_r_01_MSTR, swing1LimitDegrees=20, swing2LimitDegrees=45, twistLimitDegrees=5}
 {
  bone1Name=dyn_shirt_bk_r_01_MSTR
  bone2Name=pelvis
  swing1LimitDegrees=45
  swing1Motion=Free
  swing2LimitDegrees=45
  swing2Motion=Free
  twistLimitDegrees=45
  twistMotion=Free
 }
 {bone1Name=dyn_shirt_01_r, bone2Name=dyn_shirt_01_r_MSTR, swing1LimitDegrees=20, swing2LimitDegrees=45, twistLimitDegrees=5}
]
