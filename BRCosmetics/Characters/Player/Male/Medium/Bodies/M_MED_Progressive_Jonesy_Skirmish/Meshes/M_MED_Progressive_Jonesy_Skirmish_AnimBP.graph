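# Animation Blueprint for the Progressive Jonesy Skirmish body that layers secondary motion for shirt, belt knot and crossbow onto the pose copied from the parent mesh.
# Constraints and additive bone offsets, some toggled by crouch, jump or targeting, feed two RigidBody simulations.
# Two Control Rigs finish the pose, the second blending in while jumping or crouching.
# Each update sets the simulations' gravity override to scaled skydiving force or normal gravity.
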AnimBlueprint /BRCosmetics/Characters/Player/Male/Medium/Bodies/M_MED_Progressive_Jonesy_Skirmish/Meshes/M_MED_Progressive_Jonesy_Skirmish_AnimBP
target_skeleton=/Game/Characters/Player/Male/Medium/Bodies/M_MED_Progressive_Jonesy_Skirmish/Meshes/M_MED_Progressive_Jonesy_Skirmish_Skeleton
graphs [
 AnimGraph {nodes [
   n1 = Get bIsCrouching()
   "Transform (Modify) Bone - Bone: dyn_shirt_01_r_MSTR"!(ComponentPose, bAlphaBoolEnabled:bool=@n1, Translation="-4,3,1", Rotation="0,0,0", Scale="1,1,1") -> (Pose) {
    alphaInputType=Bool
    rotationMode=BMM_Additive
    rotationSpace=BCS_BoneSpace
    scaleMode=BMM_Additive
    scaleSpace=BCS_BoneSpace
    translationMode=BMM_Additive
    translationSpace=BCS_BoneSpace
    alphaBoolBlend {blendOption=Linear}
    boneToModify {boneName=dyn_shirt_01_r_MSTR}
    translation=<-4 3 1>
   }
   n2 = "Constraint - dyn_crossbow_01_l_MSTR (PP)"!(
    ComponentPose=Local To Component(LocalPose=Copy Pose From Mesh() {bUseAttachedParent=true})
    Alpha=1
    ConstraintWeights_0=0.5
    ConstraintWeights_1:float=0.5
   ) -> (Pose) {
    alphaBoolBlend {blendOption=Linear}
    boneToModify {boneName=dyn_crossbow_01_l_MSTR}
    constraintSetup [{targetBone {boneName=pelvis}}, {targetBone {boneName=thigh_l}}]
    constraintWeights [1,1]
   }
   n3 = "Constraint - dyn_shirt_01_r_MSTR (R)"(ComponentPose=@n2, Alpha=0, ConstraintWeights_0=1) {
    alphaBoolBlend {blendOption=Linear}
    boneToModify {boneName=dyn_shirt_01_r_MSTR}
    constraintSetup [{transformType=Rotation, targetBone {boneName=thigh_r}}]
    constraintWeights [1]
   }
   n4 = Get bIsTargeting()
   n5 = "Constraint - dyn_shirt_fr_01_r_MSTR (RR)"!(ComponentPose=@n3, bAlphaBoolEnabled:bool=NOT Boolean(A=OR Boolean(A=@n4, B=@n1)), ConstraintWeights_0=0.75, ConstraintWeights_1:float=0.25) -> (Pose) {
    alphaInputType=Bool
    alphaBoolBlend {blendInTime=0.1, blendOption=Linear, blendOutTime=0.1}
    boneToModify {boneName=dyn_shirt_fr_01_r_MSTR}
    constraintSetup [{transformType=Rotation, targetBone {boneName=thigh_twist_01_r}}, {transformType=Rotation, targetBone {boneName=thigh_r}}]
    constraintWeights [1,1]
   }
   n6 = "Constraint - dyn_belt_knot_l_MSTR (RR)"!(ComponentPose=@n5, bAlphaBoolEnabled:bool=True, ConstraintWeights_0=0, ConstraintWeights_1:float=1) -> (Pose) {
    alphaInputType=Bool
    alphaBoolBlend {blendInTime=0.1, blendOption=Linear, blendOutTime=0.1}
    boneToModify {boneName=dyn_belt_knot_l_MSTR}
    constraintSetup [{transformType=Rotation, targetBone {boneName=thigh_twist_01_r}}, {transformType=Rotation, targetBone {boneName=thigh_r}}]
    constraintWeights [1,1]
   }
   n7 = "Transform (Modify) Bone - Bone: dyn_shirt_fr_01_r_MSTR"(ComponentPose=@n6, Alpha=1, Translation="-0.000,0,0.371", Rotation="0,0,0", Scale="1,1,1") {
    rotationMode=BMM_Additive
    rotationSpace=BCS_BoneSpace
    scaleMode=BMM_Additive
    scaleSpace=BCS_BoneSpace
    translationMode=BMM_Additive
    translationSpace=BCS_BoneSpace
    alphaBoolBlend {blendOption=Linear}
    boneToModify {boneName=dyn_shirt_fr_01_r_MSTR}
    translation {z=0.371}
   }
   n8 = "Transform (Modify) Bone - Bone: dyn_shirt_fr_01_r_MSTR"!(ComponentPose=@n7, bAlphaBoolEnabled:bool=Get bIsJumping(), Translation="0,0,0", Rotation="15,0,0", Scale="1,1,1") -> (Pose) {
    alphaInputType=Bool
    rotationMode=BMM_Additive
    rotationSpace=BCS_BoneSpace
    scaleMode=BMM_Additive
    scaleSpace=BCS_BoneSpace
    translationMode=BMM_Additive
    translationSpace=BCS_BoneSpace
    alphaBoolBlend {blendOption=Linear}
    boneToModify {boneName=dyn_shirt_fr_01_r_MSTR}
    rotation {pitch=15}
   }
   n9 = "Transform (Modify) Bone - Bone: dyn_shirt_01_r_MSTR"(ComponentPose=@n8, bAlphaBoolEnabled=@n1, Translation="-4,0,2", Rotation="0,0,0", Scale="1,1,1") {
    alphaInputType=Bool
    rotationMode=BMM_Additive
    rotationSpace=BCS_BoneSpace
    scaleMode=BMM_Additive
    scaleSpace=BCS_BoneSpace
    translationMode=BMM_Additive
    translationSpace=BCS_BoneSpace
    alphaBoolBlend {blendOption=Linear}
    boneToModify {boneName=dyn_shirt_01_r_MSTR}
    translation {x=-4, z=2}
   }
   n10 = "Constraint - dyn_belt_knot_MSTR (PPPPPP)"!(
    ComponentPose=@n9
    Alpha=1
    ConstraintWeights_0=0.0011
    ConstraintWeights_1:float=0.2748
    ConstraintWeights_2:float=0.2731
    ConstraintWeights_3:float=0.1948
    ConstraintWeights_4:float=0.045
    ConstraintWeights_5:float=0.2112
   ) -> (Pose) {
    alphaBoolBlend {blendOption=Linear}
    boneToModify {boneName=dyn_belt_knot_MSTR}
    constraintSetup [
     {targetBone {boneName=dyn_shirt_fr_01_r}}
     {targetBone {boneName=spine_01}}
     {targetBone {boneName=spine_02}}
     {targetBone {boneName=spine_03}}
     {targetBone {boneName=spine_04}}
     {targetBone {boneName=pelvis}}
    ]
    constraintWeights [1,1,1,1,1,1]
   }
   n11 = "Transform (Modify) Bone - Bone: dyn_crossbow_01_l"!(ComponentPose=@n10, bAlphaBoolEnabled:bool=@n1, Translation="-1.5,-1,-1", Rotation="0,0,0", Scale="1,1,1") -> (Pose) {
    alphaInputType=Bool
    rotationMode=BMM_Additive
    translationMode=BMM_Additive
    alphaBoolBlend {blendInTime=0.1, blendOption=Linear, blendOutTime=0.1}
    boneToModify {boneName=dyn_crossbow_01_l}
    translation=<-1.5 -1 -1>
   }
   n12 = Get windVector() -> (windVector:Vector)
   n13 = RigidBody!(ComponentPose=@n11, Alpha, OverrideWorldGravity:Vector=@n12, ExternalForce) -> (Pose) {
    bClampLinearTranslationLimitToRefPose=true
    bOverrideWorldGravity=true
    overridePhysicsAsset=&/M_MED_Progressive_Jonesy_Skirmish_Cb_Physics
    simulationSpace=BaseBoneSpace
    alphaBoolBlend {blendOption=Linear}
    baseBoneRef {boneName=root}
    componentAppliedLinearAccClamp=<3000 3000 3000>
    componentLinearAccScale=<0.3 0.3 0.7>
    componentLinearVelScale=<0.3 0.3 0.7>
    overrideWorldGravity {z=-980}
   }
   n14 = "Transform (Modify) Bone - Bone: dyn_crossbow_03_l"!(ComponentPose=@n13, bAlphaBoolEnabled:bool=@n1, Translation="-0.5,-2,-0.5", Rotation="-5,0,0", Scale="1,1,1") -> (Pose) {
    alphaInputType=Bool
    translationMode=BMM_Additive
    alphaBoolBlend {blendInTime=0.1, blendOption=Linear, blendOutTime=0.1}
    boneToModify {boneName=dyn_crossbow_03_l}
    rotation {pitch=-5}
    translation=<-0.5 -2 -0.5>
   }
   n15 = RigidBody(ComponentPose=@n14, OverrideWorldGravity=@n12) {
    bClampLinearTranslationLimitToRefPose=true
    bOverrideWorldGravity=true
    overridePhysicsAsset=&/M_MED_Progressive_Jonesy_Skirmish_Physics
    simulationSpace=BaseBoneSpace
    alphaBoolBlend {blendOption=Linear}
    baseBoneRef {boneName=pelvis}
    componentAppliedLinearAccClamp=<3000 3000 3000>
    componentLinearAccScale=<0.3 0.3 0.4>
    componentLinearVelScale=<0.3 0.3 0.4>
    overrideWorldGravity {z=-980}
   }
   n16 = Control Rig(Source=Component To Local(ComponentPose=@n15)) {
    alphaBoolBlend {blendOption=Linear}
    controlRigAssetReference {blueprintRigClass=/Game/Characters/Player/Male/Medium/Bodies/M_MED_Progressive_Jonesy_Skirmish/Meshes/M_MED_Progressive_Jonesy_Skirmish_CtrlRig/M_MED_Progressive_Jonesy_Skirmish_CtrlRig.M_MED_Progressive_Jonesy_Skirmish_CtrlRig_C}
   }
   n17 = Control Rig!(Source=@n16, bAlphaBoolEnabled:bool=OR Boolean(A=Get bIsJumping(), B=Get bIsCrouching())) -> (Pose) {
    alphaInputType=Bool
    alphaBoolBlend {blendInTime=0.3, blendOption=Linear, blendOutTime=0.5}
    controlRigAssetReference {blueprintRigClass=&/M_MED_Progressive_Jonesy_Skirmish_cr.M_MED_Progressive_Jonesy_Skirmish_cr_C}
   }
   Output Pose(Result=@n17) {layerGroup=None}
   Get bIsFalling()
  ]}
 EventGraph {nodes [
   Delay(Duration=0.2)
   > ResetDynamics(InTeleportType=TeleportPhysics)
   > ResetDynamics(InTeleportType=ResetPhysics)
   Event BlueprintUpdateAnimation()
   > Sequence()
   n1 = Break Fort Anim Input Skydiving External Force(FortAnimInput_SkydivingExternalForce=Get SkydivingExternalForceInput()) -> (
    bUseSkydivingVectorForce=~
    SkydivingVectorForce=~
    HeadToPelvisDirection=~
    FloatingMultiplier=~
    FloatingAdditive=~
    DivingMultiplier=~
    DivingAdditive=~
    ParachutingMultiplier=~
    ParachutingAdditive=~
   )
   > Set windVector(windVector:Vector=SelectVector(A=vector * float(A=@n1.SkydivingVectorForce, B=0.15), B="0, 0,-980", bPickA=Get bIsSkyDiving())) -> (Output_Get:Vector)
   Event OnEndCrouching()
  ]}
]
variables [
 Jacket Float {type=number}
 windVector {type=Vector}
]
