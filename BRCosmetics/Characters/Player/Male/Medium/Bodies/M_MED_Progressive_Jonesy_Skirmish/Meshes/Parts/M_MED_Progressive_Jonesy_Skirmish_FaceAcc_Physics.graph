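# Physics asset for the face accessory of the Progressive Jonesy Skirmish male character, defining bodies and joints for its dynamic band bones.
# The left band is a five-box chain from a small master capsule on the head, with tight twist and swing limits that widen toward the tip.
# The right band is one limited capsule on the head.
# Capsules on spine, neck, head and clavicles form the upper body, joined by angularly free constraints.
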PhysicsAsset /BRCosmetics/Characters/Player/Male/Medium/Bodies/M_MED_Progressive_Jonesy_Skirmish/Meshes/Parts/M_MED_Progressive_Jonesy_Skirmish_FaceAcc_Physics
bodies {
 clavicle_l [{shapeType=Capsule, length=16.0993, radius=9.5934, center=<9 0 0>, rotation=<rot 90 0 0>}]
 clavicle_r [{shapeType=Capsule, length=16.0993, radius=9.5934, center=<-9 0 0>, rotation=<rot 90 0 0>}]
 dyn_band_l_01 [{shapeType=Box, extentX=7.9483, extentY=2.6549, extentZ=2.8711, center=<3 0 0>, rotation=<rot 0 0 2.3075>}]
 dyn_band_l_02 [{shapeType=Box, extentX=7.9306, extentY=3.2398, extentZ=3.1479, center=<3 0 0>, rotation=<rot 0 0 10.8731>}]
 dyn_band_l_03 [{shapeType=Box, extentX=9.9274, extentY=3.5385, extentZ=3.3685, center=<3.5 0 0>, rotation=<rot -0.0001 0 12.0652>}]
 dyn_band_l_04 [{shapeType=Box, extentX=11.111, extentY=3.8338, extentZ=3.1069, center=<3.5 0 0>, rotation=<rot 0 0 14.6735>}]
 dyn_band_l_05 [{shapeType=Box, extentX=9.4974, extentY=4.1957, extentZ=3.2001, center=<3.5 0 0>, rotation=<rot 0 0 14.6735>}]
 dyn_band_l_MSTR [{shapeType=Capsule, length=0.505, radius=0.505, rotation=<rot 0 -90 0>}]
 dyn_band_r_01 [{shapeType=Capsule, length=1.2832, radius=0.8697, center=<2.1849 0 0>, rotation=<rot 0 0 -90>}]
 head [{shapeType=Capsule, length=3.6389, radius=6.917, center=<2.8624 0.5728 0>, rotation=<rot 90 179.1048 174.1048>}]
 neck_01 [{shapeType=Capsule, length=15.7891, radius=8.9832, center=<-0.8309 0.2503 0>}]
 spine_03 [{shapeType=Capsule, length=16.7325, radius=12.3267, center=<3.5 -0.5 0>}]
 spine_04 [{shapeType=Capsule, length=19.3122, radius=13.3786, center=<4.5 -0.75 0>}]
 spine_05 [{shapeType=Capsule, length=24.9908, radius=12.4925, center=<3 -1 0>}]
}
constraints [
 {
  bone1Name=spine_05
  bone2Name=spine_04
  swing1LimitDegrees=45
  swing1Motion=Free
  swing2LimitDegrees=45
  swing2Motion=Free
  twistLimitDegrees=45
  twistMotion=Free
 }
 {
  bone1Name=head
  bone2Name=neck_01
  swing1LimitDegrees=45
  swing1Motion=Free
  swing2LimitDegrees=45
  swing2Motion=Free
  twistLimitDegrees=45
  twistMotion=Free
 }
 {bone1Name=dyn_band_l_01, bone2Name=dyn_band_l_MSTR, swing1LimitDegrees=20, swing2LimitDegrees=30, twistLimitDegrees=5}
 {bone1Name=dyn_band_l_02, bone2Name=dyn_band_l_01, swing1LimitDegrees=20, swing2LimitDegrees=30, twistLimitDegrees=5}
 {bone1Name=dyn_band_l_03, bone2Name=dyn_band_l_02, swing1LimitDegrees=20, swing2LimitDegrees=35, twistLimitDegrees=5}
 {bone1Name=dyn_band_l_04, bone2Name=dyn_band_l_03, swing1LimitDegrees=30, swing2LimitDegrees=40, twistLimitDegrees=5}
 {bone1Name=dyn_band_l_05, bone2Name=dyn_band_l_04, swing1LimitDegrees=45, swing2LimitDegrees=45, twistLimitDegrees=5}
 {bone1Name=dyn_band_r_01, bone2Name=head, swing1LimitDegrees=45, swing2LimitDegrees=45, twistLimitDegrees=5}
 {
  bone1Name=neck_01
  bone2Name=spine_05
  swing1LimitDegrees=45
  swing1Motion=Free
  swing2LimitDegrees=45
  swing2Motion=Free
  twistLimitDegrees=45
  twistMotion=Free
 }
 {
  bone1Name=dyn_band_l_MSTR
  bone2Name=head
  swing1LimitDegrees=45
  swing1Motion=Free
  swing2LimitDegrees=45
  swing2Motion=Free
  twistLimitDegrees=45
  twistMotion=Free
 }
 {
  bone1Name=spine_04
  bone2Name=spine_03
  swing1LimitDegrees=45
  swing1Motion=Free
  swing2LimitDegrees=45
  swing2Motion=Free
  twistLimitDegrees=45
  twistMotion=Free
 }
 {
  bone1Name=clavicle_r
  bone2Name=spine_05
  swing1LimitDegrees=45
  swing1Motion=Free
  swing2LimitDegrees=45
  swing2Motion=Free
  twistLimitDegrees=45
  twistMotion=Free
 }
 {
  bone1Name=clavicle_l
  bone2Name=spine_05
  swing1LimitDegrees=45
  swing1Motion=Free
  swing2LimitDegrees=45
  swing2Motion=Free
  twistLimitDegrees=45
  twistMotion=Free
 }
]
