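# Character-part AnimBP for Progressive Jonesy Skirmish that adds secondary motion to face accessories, hair and a left band chain.
# It copies the parent pose, rotates the band master and sim-space root, then runs two RigidBody nodes, five sphere-limited AnimDynamics links and a final LookAt.
# Update logic feeds these through deprecated state-based gravity/wind and sim-space helpers plus random rotator noise.
# GravityOverride_Hair, the hair RigidBody's external force, is never set here.
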AnimBlueprint /BRCosmetics/Characters/Player/Male/Medium/Bodies/M_MED_Progressive_Jonesy_Skirmish/Meshes/Parts/M_MED_Progressive_Jonesy_Skirmish_AnimBP
target_skeleton=&/M_MED_Progressive_Jonesy_Skirmish_Skeleton
graphs [
 AnimGraph {nodes [
   n1 = "Transform (Modify) Bone - Bone: dyn_band_l_MSTR"(
    ComponentPose=Local To Component(LocalPose=Copy Pose From Mesh() {bUseAttachedParent=true})
    Alpha=0.5
    Translation="0,0,0"
    Rotation=Get HairRotator_01() -> (HairRotator_01:Rotator)
    Scale="1,1,1"
   ) {
    rotationMode=BMM_Additive
    rotationSpace=BCS_WorldSpace
    translationSpace=BCS_BoneSpace
    alphaBoolBlend {blendOption=Linear}
    alphaScaleBiasClamp {bInterpResult=true, interpSpeedDecreasing=2, interpSpeedIncreasing=2}
    boneToModify {boneName=dyn_band_l_MSTR}
   }
   n2 = "Transform (Modify) Bone - Bone: dyn_root_simspace"(ComponentPose=@n1, Alpha=1, Translation="0,0,0", Rotation=Get sim_space_rotation() -> (sim_space_rotation:Rotator), Scale="1,1,1") {
    rotationMode=BMM_Replace
    rotationSpace=BCS_WorldSpace
    translationSpace=BCS_BoneSpace
    alphaBoolBlend {blendOption=Linear}
    alphaScaleBiasClamp {bInterpResult=true, interpSpeedDecreasing=2, interpSpeedIncreasing=2}
    boneToModify {boneName=dyn_root_simspace}
   }
   n3 = RigidBody!(ComponentPose=@n2, Alpha, OverrideWorldGravity:Vector=Get GravityOverride(), ExternalForce) -> (Pose) {
    bAllowPoseScalePropagation=false
    bClampLinearTranslationLimitToRefPose=true
    bOverrideWorldGravity=true
    overridePhysicsAsset=&/M_MED_Progressive_Jonesy_Skirmish_FaceAcc_Physics
    simulationSpace=BaseBoneSpace
    alphaBoolBlend {blendOption=Linear}
    baseBoneRef {boneName=dyn_root_simspace}
    componentAppliedLinearAccClamp=<3000 3000 3000>
    componentLinearAccScale=<0.3 0.3 0.3>
    componentLinearVelScale=<0.3 0.3 0.3>
    overrideWorldGravity {z=-980}
    simSpaceSettings {
     maxAngularAcceleration=1000
     maxAngularVelocity=3
     maxLinearAcceleration=700
     maxLinearVelocity=500
     velocityScaleZ=0
     worldAlpha=1
     externalLinearDragV=<1 1>
    }
   }
   n4 = RigidBody!(ComponentPose=@n3, Alpha, ExternalForce=Get GravityOverride_Hair() -> (GravityOverride_Hair:Vector)) -> (Pose) {
    bClampLinearTranslationLimitToRefPose=true
    bOverrideWorldGravity=true
    overridePhysicsAsset=&/M_MED_Progressive_Jonesy_Skirmish_H_P
    simulationSpace=BaseBoneSpace
    alphaBoolBlend {blendOption=Linear}
    baseBoneRef {boneName=neck_01}
    componentAppliedLinearAccClamp=<3000 3000 3000>
    componentLinearAccScale=<0.2 0.2 0.2>
    componentLinearVelScale=<0.2 0.2 0.2>
    overrideWorldGravity {z=-980}
    simSpaceSettings {
     maxAngularAcceleration=1000
     maxAngularVelocity=3
     maxLinearAcceleration=700
     maxLinearVelocity=500
     velocityScaleZ=0
     worldAlpha=0.5
     externalLinearDragV=<0.5 0.5>
    }
   }
   n5 = Get GravityOverride()
   n6 = Anim Dynamics - Bone: dyn_band_l_01!(ComponentPose=@n4, Alpha=1, GravityOverride:Vector=@n5, ExternalForce="0,0,0") -> (Pose) {
    angularDampingOverride=0.9999
    angularSpringConstant=500
    bAngularSpring=true
    bGravityOverrideInSimSpace=true
    bLinearSpring=true
    bOverrideAngularDamping=true
    bUseGravityOverride=true
    bUseSphericalLimits=true
    linearSpringConstant=50
    simulationSpace=BoneRelative
    alphaBoolBlend {blendOption=Linear}
    boundBone {boneName=dyn_band_l_01}
    componentAppliedLinearAccClamp=<1500 1500 1500>
    componentLinearAccScale=<0.1 0.5 0.5>
    componentLinearVelScale=<0.1 0.5 0.5>
    gravityOverride {z=-980}
    physicsBodyDefinitions [{
      collisionType=CustomSphere
      sphereCollisionRadius=2
      boundBone {boneName=dyn_band_l_01}
      boxExtents=<5 20 20>
      constraintSetup {angularLimitsMax=<1 15 20>, angularLimitsMin=<-1 -15 -20>, angularTarget=<0 0 0>}
      localJointOffset=<5 0 0>
     }]
    relativeSpaceBone {boneName=root}
    sphericalLimits [{limitRadius=10, drivingBone {boneName=head}, sphereLocalOffset=<0 3 0>}]
   }
   n7 = Anim Dynamics - Bone: dyn_band_l_02!(ComponentPose=@n6, Alpha=1, GravityOverride:Vector=@n5, ExternalForce="0,0,0") -> (Pose) {
    angularDampingOverride=0.9999
    angularSpringConstant=500
    bAngularSpring=true
    bGravityOverrideInSimSpace=true
    bLinearSpring=true
    bOverrideAngularDamping=true
    bUseGravityOverride=true
    bUseSphericalLimits=true
    linearSpringConstant=50
    simulationSpace=BoneRelative
    alphaBoolBlend {blendOption=Linear}
    boundBone {boneName=dyn_band_l_02}
    componentAppliedLinearAccClamp=<3000 3000 3000>
    componentLinearAccScale=<0.1 0.8 0.8>
    componentLinearVelScale=<0.1 0.8 0.8>
    gravityOverride {z=-980}
    physicsBodyDefinitions [{
      collisionType=CustomSphere
      sphereCollisionRadius=2
      boundBone {boneName=dyn_band_l_02}
      boxExtents=<5 20 20>
      constraintSetup {angularLimitsMax=<1 30 25>, angularLimitsMin=<-1 -35 -25>, angularTarget=<0 0 0>}
      localJointOffset=<5 0 0>
     }]
    relativeSpaceBone {boneName=root}
    sphericalLimits [
     {limitRadius=10, drivingBone {boneName=neck_02}, sphereLocalOffset=<0 3 0>}
     {limitRadius=10, drivingBone {boneName=neck_01}, sphereLocalOffset=<0 0 0>}
     {limitRadius=10, drivingBone {boneName=spine_05}, sphereLocalOffset=<0 0 0>}
    ]
   }
   n8 = Anim Dynamics - Bone: dyn_band_l_03!(ComponentPose=@n7, Alpha=1, GravityOverride:Vector=@n5, ExternalForce="0,0,0") -> (Pose) {
    angularDampingOverride=0.9999
    angularSpringConstant=500
    bAngularSpring=true
    bGravityOverrideInSimSpace=true
    bLinearSpring=true
    bOverrideAngularDamping=true
    bUseGravityOverride=true
    bUseSphericalLimits=true
    linearSpringConstant=50
    simulationSpace=BoneRelative
    alphaBoolBlend {blendOption=Linear}
    boundBone {boneName=dyn_band_l_03}
    componentAppliedLinearAccClamp=<3000 3000 3000>
    componentLinearAccScale=<0.5 1.2 1.2>
    componentLinearVelScale=<0.5 1.2 1.2>
    gravityOverride {z=-980}
    physicsBodyDefinitions [{
      collisionType=CustomSphere
      sphereCollisionRadius=2
      boundBone {boneName=dyn_band_l_03}
      boxExtents=<5 20 20>
      constraintSetup {angularLimitsMax=<1 30 25>, angularLimitsMin=<-1 -35 -25>, angularTarget=<0 0 0>}
      localJointOffset=<5 0 0>
     }]
    relativeSpaceBone {boneName=root}
    sphericalLimits [
     {limitRadius=10, drivingBone {boneName=neck_02}, sphereLocalOffset=<0 3 0>}
     {limitRadius=10, drivingBone {boneName=neck_01}, sphereLocalOffset=<0 0 0>}
     {limitRadius=10, drivingBone {boneName=spine_05}, sphereLocalOffset=<0 0 0>}
     {limitRadius=12, drivingBone {boneName=upperarm_l}, sphereLocalOffset=<0 0 0>}
     {limitRadius=12, drivingBone {boneName=upperarm_r}, sphereLocalOffset=<0 0 0>}
    ]
   }
   n9 = Anim Dynamics - Bone: dyn_band_l_04!(ComponentPose=@n8, Alpha=1, GravityOverride:Vector=@n5, ExternalForce="0,0,0") -> (Pose) {
    angularDampingOverride=0.9999
    angularSpringConstant=500
    bAngularSpring=true
    bGravityOverrideInSimSpace=true
    bLinearSpring=true
    bOverrideAngularDamping=true
    bUseGravityOverride=true
    bUseSphericalLimits=true
    linearSpringConstant=50
    simulationSpace=BoneRelative
    alphaBoolBlend {blendOption=Linear}
    boundBone {boneName=dyn_band_l_04}
    componentAppliedLinearAccClamp=<3000 3000 3000>
    componentLinearAccScale=<0.5 1.2 1.2>
    componentLinearVelScale=<0.5 1.2 1.2>
    gravityOverride {z=-980}
    physicsBodyDefinitions [{
      collisionType=CustomSphere
      sphereCollisionRadius=2
      boundBone {boneName=dyn_band_l_04}
      boxExtents=<5 20 20>
      constraintSetup {angularLimitsMax=<1 30 25>, angularLimitsMin=<-1 -35 -25>, angularTarget=<0 0 0>}
      localJointOffset=<5 0 0>
     }]
    relativeSpaceBone {boneName=root}
    sphericalLimits [
     {limitRadius=5, drivingBone {boneName=neck_02}, sphereLocalOffset=<0 3 0>}
     {limitRadius=5, drivingBone {boneName=neck_01}, sphereLocalOffset=<0 0 0>}
     {limitRadius=10, drivingBone {boneName=spine_05}, sphereLocalOffset=<0 0 0>}
     {limitRadius=12, drivingBone {boneName=clavicle_r}, sphereLocalOffset=<0 0 0>}
     {limitRadius=12, drivingBone {boneName=clavicle_l}, sphereLocalOffset=<0 0 0>}
    ]
   }
   n10 = Anim Dynamics - Bone: dyn_band_l_05!(ComponentPose=@n9, Alpha=1, GravityOverride:Vector=@n5, ExternalForce="0,0,0") -> (Pose) {
    angularDampingOverride=0.9999
    angularSpringConstant=500
    bAngularSpring=true
    bGravityOverrideInSimSpace=true
    bLinearSpring=true
    bOverrideAngularDamping=true
    bUseGravityOverride=true
    bUseSphericalLimits=true
    linearSpringConstant=50
    simulationSpace=BoneRelative
    alphaBoolBlend {blendOption=Linear}
    boundBone {boneName=dyn_band_l_05}
    componentAppliedLinearAccClamp=<3000 3000 3000>
    componentLinearAccScale=<0.5 1.2 1.2>
    componentLinearVelScale=<0.5 1.2 1.2>
    gravityOverride {z=-980}
    physicsBodyDefinitions [{
      collisionType=CustomSphere
      sphereCollisionRadius=2
      boundBone {boneName=dyn_band_l_05}
      boxExtents=<5 20 20>
      constraintSetup {angularLimitsMax=<1 30 25>, angularLimitsMin=<-1 -35 -25>, angularTarget=<0 0 0>}
      localJointOffset=<5 0 0>
     }]
    relativeSpaceBone {boneName=root}
    sphericalLimits [
     {limitRadius=5, drivingBone {boneName=neck_02}, sphereLocalOffset=<-6 3 0>}
     {limitRadius=5, drivingBone {boneName=neck_01}, sphereLocalOffset=<0 0 0>}
     {limitRadius=10, drivingBone {boneName=spine_05}, sphereLocalOffset=<0 0 0>}
     {limitRadius=12, drivingBone {boneName=clavicle_r}, sphereLocalOffset=<0 0 0>}
     {limitRadius=12, drivingBone {boneName=clavicle_l}, sphereLocalOffset=<0 0 0>}
    ]
   }
   n11 = Look At - Bone: dyn_band_l_MSTR(ComponentPose=@n10, Alpha=1) {
    alphaBoolBlend {blendOption=Linear}
    boneToModify {boneName=dyn_band_l_MSTR}
    lookAtLocation=<-8.8436 -77.6306 37.9368>
    lookAtTarget {boneReference {boneName=spine_05}}
    lookAt_Axis {axis=<1 0>}
   }
   Output Pose(Result=Component To Local(ComponentPose=@n11)) {layerGroup=None}
  ]}
 EventGraph {nodes [
   n1 = Event BlueprintUpdateAnimation()
   > n2 = Sequence!(execute) -> (then_0, then_1, then_2:Exec)
   > n3 = Gravity Override State Function DEPRECATED(CharacterPartAnimInstance=Self-Reference(), GameTimeInSeconds=GetGameTimeInSeconds(), GravityOverrideParams=Get Gravity Override Params())
   > n4 = Set GravityOverride(GravityOverride=@n3)
   > Set GravityOverride_FrontHair(GravityOverride_FrontHair:Vector=vector * vector(A:Vector=@n4, B:Vector="0.4,0.4,2") -> (ReturnValue:Vector) {is=title K2Node_CommutativeAssociativeBinaryOperator}) -> (Output_Get:Vector)
   n5 = ComplexSimSpaceFunction DEPRECATED!(
    execute=@n2.then_1
    CharacterPartAnimInstance:Custom Character Part Anim Instance Object Reference=Self-Reference()
    InterpSpeedMax:double=9
    PawnSpeedMax:double=20
    RInterpErrorTolerance:double=1
    DeltaTime:double=@n1.DeltaTimeX
    DebugDraw:bool=false
    sim_space_rotation_in:Rotator=Get sim_space_rotation()
    inComponentSpaceIn:bool=Get InComponentSpace() -> (InComponentSpace:bool)
    inComponentSpaceLastFrameIn:bool=Get InComponentSpaceLastFrame() -> (InComponentSpaceLastFrame:bool)
    lockedOnSimSpaceIn:bool=Get LockedOnSimSpace() -> (LockedOnSimSpace:bool)
    InterpSpeedIn:double=Get InterpSpeed() -> (InterpSpeed:double)
   ) -> (
    then
    sim_space_rotation:Rotator
    inComponentSpaceOut:bool
    inComponentSpaceLastFrameOut:bool
    lockedOnSimSpaceOut:bool
    InterpSpeedOut:double
   )
   > Set sim_space_rotation(sim_space_rotation:Rotator=@n5) -> (Output_Get:Rotator)
   > Set InComponentSpace(InComponentSpace:bool=@n5.inComponentSpaceOut) -> (Output_Get:bool)
   > Set InComponentSpaceLastFrame(InComponentSpaceLastFrame:bool=@n5.inComponentSpaceLastFrameOut) -> (Output_Get:bool)
   > Set LockedOnSimSpace(LockedOnSimSpace:bool=@n5.lockedOnSimSpaceOut) -> (Output_Get:bool)
   > Set InterpSpeed(InterpSpeed:double=@n5.InterpSpeedOut) -> (Output_Get:double)
   n6 = RandomRotatorFunction!(
    execute=@n2.then_2
    GameTimeInSeconds:double=GetGameTimeInSeconds()
    PerlinScale:double=1
    Frequency:Rotator="2.3,1.5, 0"
    Amplitude:Rotator="75,60,0"
    Offset:Rotator="0,-25,0"
    TimeOffset:double=0
   ) -> (then, RandomRotator:Rotator)
   > Set HairRotator_01(HairRotator_01:Rotator=@n6) -> (Output_Get:Rotator)
   n7 = Get bRigidBodyEnabled()
   > Set RBAN Off(RBAN Off:double=SelectFloat(A=1, B=0, bPickA=NOT Boolean(A=@n7))) -> (Output_Get:double)
   Event BlueprintInitializeAnimation()
   > Set Skel_Mesh(Skel_Mesh=GetOwnerSkeletalMeshComponent())
   Self-Reference()
  ]}
]
variables [
 Body_Mesh {type=SkeletalMeshComponent}
 compLinearAccScale {type=Vector}
 compLinearVelScale {type=Vector}
 finalGravity {type=Vector}
 Gravity Override Params {
  type=GravityOverrideParamsStruct
  default {
   emoteOrMelee {gravityOverride=<0 0 -15>, windAmplitude=<50 -200 35>, windFrequency=<0.15 0.7 0.3>, windOffset=<0 -230 -50>}
   front_End {gravityOverride=<0 0 -15>, windAmplitude=<50 -200 35>, windFrequency=<0.15 0.7 0.3>, windOffset=<0 -230 -50>}
   onGround_Moving {gravityOverride=<0 0 -15>, windAmplitude=<50 -200 35>, windFrequency=<0.15 0.7 0.3>, windOffset=<0 -230 -50>}
   onGround_Standing {gravityOverride=<0 0 -15>, windAmplitude=<50 -200 35>, windFrequency=<0.15 0.7 0.3>, windOffset=<0 -230 -50>}
   skydive_DivingDown {gravityOverride=<0 0 0>}
   skydive_FloatingDown {gravityOverride=<0 0 0>}
   skydive_GoingUp {gravityOverride=<0 0 -100>}
   skydive_Parachuting {gravityOverride=<0 0 0>}
   swimming {gravityOverride=<0 0 0>}
  }
 }
 GravityOverride {type=Vector}
 GravityOverride_FrontHair {type=Vector}
 GravityOverride_Hair {type=Vector}
 HairRotator_01 {type=Rotator}
 HairVector_01 {type=Vector}
 Head_Mesh {type=SkeletalMeshComponent}
 headTwist {type=number}
 headUp {type=number}
 InComponentSpace {type=boolean}
 InComponentSpaceLastFrame {type=boolean}
 InterpSpeed {type=number}
 LockedOnSimSpace {type=boolean}
 Pawn Owner {type=PlayerPawn_Athena_C}
 RBAN Off {type=number}
 sim_space_rotation {type=Rotator}
 Skel_Mesh {type=SkeletalMeshComponent}
 windVector {type=Vector}
]
